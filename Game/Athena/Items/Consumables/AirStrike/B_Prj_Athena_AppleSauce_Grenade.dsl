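; Airstrike (AppleSauce) grenade projectile that marks where a rocket barrage falls.
; On stopping or entering water, it spawns a zone actor and warning smoke and sound, then after a delay a RocketSpawner overhead.
; The grenade later shrinks via timeline and destroys itself, taking the zone with it.
; It also holds a self-refiring downward rocket loop with jittered grid or radial placement, which the landing flow here never calls.
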(event EventBeginPlay
  ;[1 1 1] BeginPlay
  (|CallParentFunction)
  (Variables|Default|SetPlayerTransformOnFire (Transformation|GetActorTransform (Variables|Actor|GetInstigator)))
  (Components|DestroyComponent (Variables|B_Prj_Grenade|GetEffect_Distance)))

(event EventOnExploded
  (params (HitActors :type Array%20of%20Actor%20Object%20References) (HitResults :type Array%20of%20Hit%20Result%20Structures))
  ;[1 1 1] OnExploded - destroy actor
  (Actor|DestroyActor))

(event OnEnterWaterEvent(WaterInteractionComponent)
  (params (WaterBody :type Fort%20Water%20Body%20Actor%20Object%20Reference) (WaterInteractionComponent :type Fort%20Water%20Interaction%20Component%20Object%20Reference) (bIsFirstBody :type Boolean))
  (bind _returnvalue (Transformation|GetActorLocation))
  (Utilities|FlowControl|DoOnce false
    (:Completed
      (Utilities|FlowControl|Sequence
        (:then_0
          (Components|Activation|Deactivate (Variables|Default|GetRotatingMovement))
          (bind _returnvalue_2 (Building|SpawnBuildingGameplayActor "/Game/Athena/Items/Consumables/AirStrike/BGA_AppleSauce_Zone.BGA_AppleSauce_Zone_C" (Math|Transform|MakeTransform _returnvalue) (Game|GetInstigator)))
          (Variables|Default|SetZoneBGA _returnvalue_2)
          ;[1 1 1] Spawn Smoke Plume FX and Bus Horn
          ;;[1 1 1] BWayland: Moved this to a multicast event since OnEnterWater only runs on server
          (CallFunction|PlayWarningFX)
          (Utilities|FlowControl|Sequence
            (:then_0
              (Utilities|FlowControl|Delay (Math|Conversions|ToFloat(Integer) (Math|Float|Truncate (ScalableFloat|GetValueatLevel (Variables|Default|GetInitialDelay)))))
              (Building|SpawnBuildingGameplayActor "/Game/Athena/Items/Consumables/AirStrike/B_BGA_AppleSauce_RocketSpawner.B_BGA_AppleSauce_RocketSpawner_C" (Math|Transform|MakeTransform (Math|Vector|vector+vector _returnvalue (Math|Vector|MakeVector 0.0 0.0 (ScalableFloat|GetValueatLevel (Variables|Default|GetStartingHeight))))) (Variables|Actor|GetInstigator)))
            (:then_1
              (Utilities|FlowControl|Delay 5.0)
              (bind (_direction _newtrack_0) (Timeline|Timeline_0 Play)
                (:Update
                  (Transformation|SetRelativeScale3D (Variables|B_Prj_Grenade|GetMesh) (Math|Vector|Lerp(Vector) "1.100000,1.100000,1.100000" "0.050000,0.050000,0.050000" _newtrack_0)))
                (:Finished
                  (Actor|DestroyActor))))))
        (:then_1
          (Utilities|FlowControl|Delay 2.0)
          (Utilities|Time|ClearandInvalidateTimerbyHandle (Variables|Default|GetFuseIndicatorTimer)))))))

(event EventOnStop
  (params (Hit_bBlockingHit :type Boolean) (Hit_bInitialOverlap :type Boolean) (Hit_Time :type Float%20%28single-precision%29) (Hit_Distance :type Float%20%28single-precision%29) (Hit_Location :type Vector :default "0, 0, 0") (Hit_ImpactPoint :type Vector :default "0, 0, 0") (Hit_Normal :type Vector :default "0, 0, 0") (Hit_ImpactNormal :type Vector :default "0, 0, 0") (Hit_PhysMat :type Physical%20Material%20Object%20Reference) (Hit_HitActor :type Actor%20Object%20Reference) (Hit_HitComponent :type Primitive%20Component%20Object%20Reference) (Hit_HitBoneName :type Name :default "None") (Hit_BoneName :type Name :default "None") (Hit_HitItem :type Integer) (Hit_ElementIndex :type Integer) (Hit_FaceIndex :type Integer) (Hit_TraceStart :type Vector :default "0, 0, 0") (Hit_TraceEnd :type Vector :default "0, 0, 0"))
  (bind _returnvalue (Transformation|GetActorLocation))
  (Utilities|FlowControl|DoOnce false
    (:Completed
      (Utilities|FlowControl|Sequence
        (:then_0
          (Components|Activation|Deactivate (Variables|Default|GetRotatingMovement))
          (bind _returnvalue_2 (Building|SpawnBuildingGameplayActor "/Game/Athena/Items/Consumables/AirStrike/BGA_AppleSauce_Zone.BGA_AppleSauce_Zone_C" (Math|Transform|MakeTransform _returnvalue) (Game|GetInstigator)))
          (Variables|Default|SetZoneBGA _returnvalue_2)
          ;[1 1 1] Spawn Smoke Plume FX and Bus Horn
          ;;[1 1 1] BWayland: Moved this to a multicast event since OnEnterWater only runs on server
          (CallFunction|PlayWarningFX)
          (Utilities|FlowControl|Sequence
            (:then_0
              (Utilities|FlowControl|Delay (Math|Conversions|ToFloat(Integer) (Math|Float|Truncate (ScalableFloat|GetValueatLevel (Variables|Default|GetInitialDelay)))))
              (Building|SpawnBuildingGameplayActor "/Game/Athena/Items/Consumables/AirStrike/B_BGA_AppleSauce_RocketSpawner.B_BGA_AppleSauce_RocketSpawner_C" (Math|Transform|MakeTransform (Math|Vector|vector+vector _returnvalue (Math|Vector|MakeVector 0.0 0.0 (ScalableFloat|GetValueatLevel (Variables|Default|GetStartingHeight))))) (Variables|Actor|GetInstigator)))
            (:then_1
              (Utilities|FlowControl|Delay 5.0)
              (Timeline|Timeline_0 Play))))
        (:then_1
          (Utilities|FlowControl|Delay 2.0)
          (Utilities|Time|ClearandInvalidateTimerbyHandle (Variables|Default|GetFuseIndicatorTimer)))))))

(event Custom|ReSpawnReticle
  (Variables|Default|SetStartLocationRocket (CallFunction|CalcSpawnLocOfNextRocket))
  (Variables|Default|SetStartRotationRocket (CallFunction|CalcSpawnRotforAllRockets))
  ;[1 1 1] Fire Rocket Function
  (CallFunction|FireAirStrikeRocket (Variables|Default|GetStartLocationRocket) (Variables|Default|GetStartRotationRocket) (Variables|Default|GetECRocket) (ScalableFloat|GetValueatLevel (Variables|Default|GetRocketSpeed)) (Variables|Default|GetRocketGravityScale))
  ;[1 1 1] Delay between rockets
  (Utilities|FlowControl|Delay (+ (ScalableFloat|GetValueatLevel (Variables|Default|GetTimeBetweenRockets)) (Math|Random|RandomFloatinRange (* (ScalableFloat|GetValueatLevel (Variables|Default|GetTimeBetweenRocketsRandomize)) -1.0) (ScalableFloat|GetValueatLevel (Variables|Default|GetTimeBetweenRocketsRandomize)))))
  ;[1 1 1] Increment Counters and Fire again if not done
  (Math|Float|DecrementFloat (Variables|Default|GetPlacementMultiplier)
    (:%20%20
      (bind _result (Math|Integer|DecrementInt (Variables|Default|GetRocketsFiredNumber))
        (:%20%20
          (bind _result_2 (Math|Integer|IncrementInt (Variables|Default|GetCellNumber))
            (:%20%20
              (if (>= _result_2 (Variables|Default|GetCellsPerRow))
                (Variables|Default|SetCellNumber 0)
                (bind _result_3 (Math|Integer|IncrementInt (Variables|Default|GetRowNumber))
                  (:%20%20
                    (if (>= _result_3 (Math|Float|Truncate (ScalableFloat|GetValueatLevel (Variables|Default|GetRowSize))))
                      (Variables|Default|SetRowNumber 0)
                      (bind _result_4 (Math|Integer|IncrementInt (Variables|Default|GetClusterNumber))
                        (:%20%20
                          (if (not (>= _result_4 (Math|Float|Truncate (ScalableFloat|GetValueatLevel (Variables|Default|GetClusterAmount)))))
                            ;[1 1 1] When group is done, start over and fire next group
                            (Utilities|FlowControl|Delay (ScalableFloat|GetValueatLevel (Variables|Default|GetTimeBetweenGroups)))
                            (CallFunction|StartRockets)
                            (Utilities|FlowControl|Delay 1.0)
                            (Development|SetHiddeninGame (Variables|B_Prj_Grenade|GetMesh) true false)
                            (Components|Activation|Deactivate (Variables|B_Prj_Grenade|GetGrenadeFuse_AudioComponent)))))
                      (elif (> _result 0)
                        (CallFunction|RefireAirStrikeRocket)))
                    (Utilities|FlowControl|Delay 3.0)
                    (Utilities|FlowControl|Delay 3.0)
                    (Actor|DestroyActor)))))))))))

(event Custom|RefireAirStrikeRocket
  (Variables|Default|SetStartLocationRocket (CallFunction|CalcSpawnLocOfNextRocket))
  (Variables|Default|SetStartRotationRocket (CallFunction|CalcSpawnRotforAllRockets))
  ;[1 1 1] Fire Rocket Function
  (CallFunction|FireAirStrikeRocket (Variables|Default|GetStartLocationRocket) (Variables|Default|GetStartRotationRocket) (Variables|Default|GetECRocket) (ScalableFloat|GetValueatLevel (Variables|Default|GetRocketSpeed)) (Variables|Default|GetRocketGravityScale))
  ;[1 1 1] Delay between rockets
  (Utilities|FlowControl|Delay (+ (ScalableFloat|GetValueatLevel (Variables|Default|GetTimeBetweenRockets)) (Math|Random|RandomFloatinRange (* (ScalableFloat|GetValueatLevel (Variables|Default|GetTimeBetweenRocketsRandomize)) -1.0) (ScalableFloat|GetValueatLevel (Variables|Default|GetTimeBetweenRocketsRandomize)))))
  ;[1 1 1] Increment Counters and Fire again if not done
  (Math|Float|DecrementFloat (Variables|Default|GetPlacementMultiplier)
    (:%20%20
      (bind _result (Math|Integer|DecrementInt (Variables|Default|GetRocketsFiredNumber))
        (:%20%20
          (bind _result_2 (Math|Integer|IncrementInt (Variables|Default|GetCellNumber))
            (:%20%20
              (if (>= _result_2 (Variables|Default|GetCellsPerRow))
                (Variables|Default|SetCellNumber 0)
                (bind _result_3 (Math|Integer|IncrementInt (Variables|Default|GetRowNumber))
                  (:%20%20
                    (if (>= _result_3 (Math|Float|Truncate (ScalableFloat|GetValueatLevel (Variables|Default|GetRowSize))))
                      (Variables|Default|SetRowNumber 0)
                      (bind _result_4 (Math|Integer|IncrementInt (Variables|Default|GetClusterNumber))
                        (:%20%20
                          (if (not (>= _result_4 (Math|Float|Truncate (ScalableFloat|GetValueatLevel (Variables|Default|GetClusterAmount)))))
                            ;[1 1 1] When group is done, start over and fire next group
                            (Utilities|FlowControl|Delay (ScalableFloat|GetValueatLevel (Variables|Default|GetTimeBetweenGroups)))
                            (CallFunction|StartRockets)
                            (Utilities|FlowControl|Delay 1.0)
                            (Development|SetHiddeninGame (Variables|B_Prj_Grenade|GetMesh) true false)
                            (Components|Activation|Deactivate (Variables|B_Prj_Grenade|GetGrenadeFuse_AudioComponent)))))
                      (elif (> _result 0)
                        (CallFunction|RefireAirStrikeRocket)))
                    (Utilities|FlowControl|Delay 3.0)
                    (Utilities|FlowControl|Delay 3.0)
                    (Actor|DestroyActor)))))))))))

(event Custom|StartRockets
  (Variables|Default|SetFiringRockets true)
  (Variables|Default|SetCellNumber 0)
  (Variables|Default|SetRowNumber 0)
  (Variables|Default|SetRocketsFiredNumber (Math|Float|Truncate (ScalableFloat|GetValueatLevel (Variables|Default|GetRocketAmount))))
  (Variables|Default|SetStartLocationRocket (CallFunction|CalcSpawnLocOfNextRocket))
  (Variables|Default|SetStartRotationRocket (CallFunction|CalcSpawnRotforAllRockets))
  ;[1 1 1] Fire Rocket Function
  (CallFunction|FireAirStrikeRocket (Variables|Default|GetStartLocationRocket) (Variables|Default|GetStartRotationRocket) (Variables|Default|GetECRocket) (ScalableFloat|GetValueatLevel (Variables|Default|GetRocketSpeed)) (Variables|Default|GetRocketGravityScale))
  ;[1 1 1] Delay between rockets
  (Utilities|FlowControl|Delay (+ (ScalableFloat|GetValueatLevel (Variables|Default|GetTimeBetweenRockets)) (Math|Random|RandomFloatinRange (* (ScalableFloat|GetValueatLevel (Variables|Default|GetTimeBetweenRocketsRandomize)) -1.0) (ScalableFloat|GetValueatLevel (Variables|Default|GetTimeBetweenRocketsRandomize)))))
  ;[1 1 1] Increment Counters and Fire again if not done
  (Math|Float|DecrementFloat (Variables|Default|GetPlacementMultiplier)
    (:%20%20
      (bind _result (Math|Integer|DecrementInt (Variables|Default|GetRocketsFiredNumber))
        (:%20%20
          (bind _result_2 (Math|Integer|IncrementInt (Variables|Default|GetCellNumber))
            (:%20%20
              (if (>= _result_2 (Variables|Default|GetCellsPerRow))
                (Variables|Default|SetCellNumber 0)
                (bind _result_3 (Math|Integer|IncrementInt (Variables|Default|GetRowNumber))
                  (:%20%20
                    (if (>= _result_3 (Math|Float|Truncate (ScalableFloat|GetValueatLevel (Variables|Default|GetRowSize))))
                      (Variables|Default|SetRowNumber 0)
                      (bind _result_4 (Math|Integer|IncrementInt (Variables|Default|GetClusterNumber))
                        (:%20%20
                          (if (not (>= _result_4 (Math|Float|Truncate (ScalableFloat|GetValueatLevel (Variables|Default|GetClusterAmount)))))
                            ;[1 1 1] When group is done, start over and fire next group
                            (Utilities|FlowControl|Delay (ScalableFloat|GetValueatLevel (Variables|Default|GetTimeBetweenGroups)))
                            (CallFunction|StartRockets)
                            (Utilities|FlowControl|Delay 1.0)
                            (Development|SetHiddeninGame (Variables|B_Prj_Grenade|GetMesh) true false)
                            (Components|Activation|Deactivate (Variables|B_Prj_Grenade|GetGrenadeFuse_AudioComponent)))))
                      (elif (> _result 0)
                        (CallFunction|RefireAirStrikeRocket)))
                    (Utilities|FlowControl|Delay 3.0)
                    (Utilities|FlowControl|Delay 3.0)
                    (Actor|DestroyActor)))))))))))

(event Custom|PlayWarningFX
  (bind _returnvalue (Transformation|GetActorLocation))
  (Utilities|FlowControl|DoOnce false
    (:Completed
      (Networking|CanExecuteCosmeticEvents
        (:True
          (Audio|PlaySoundatLocation (Variables|Default|GetWarningSound) _returnvalue)
          (Niagara|SpawnSystematLocation (Variables|Default|GetSpawn%20Smoke) _returnvalue)
          (Utilities|FlowControl|Delay 0.2)
          (bind _waterfuseaudiocomponent (Variables|Default|GetWaterFuseAudioComponent)
            (:then
              (Audio|Components|Audio|FadeOut _waterfuseaudiocomponent 0.2 0.0 "Linear" :!class /Script/Engine.AudioComponent))
            (:else)))))))

(event EventDestroyed
  (Actor|DestroyActor (Variables|Default|GetZoneBGA)))

(fn CalcSpawnLocOfNextRocket
  (return (+ (select (Variables|Default|GetGridFire) (+ (* (Math|Vector|RotateVectorAroundAxis (Variables|Default|GetAirStrikeLineDirection) 9e+01 "0, 0,1.000000") (* (Variables|Default|GetRowNumber) (ScalableFloat|GetValueatLevel (Variables|Default|GetDistanceBetweenRockets)))) (+ (* (Variables|Default|GetAirStrikeLineDirection) (* (Variables|Default|GetCellNumber) (ScalableFloat|GetValueatLevel (Variables|Default|GetDistanceBetweenRockets)))) (Math|Vector|MakeVector (.x (Variables|Default|GetAirStrikeMidPoint)) (.y (Variables|Default|GetAirStrikeMidPoint)) (+ (.z (Variables|Default|GetAirStrikeMidPoint)) (ScalableFloat|GetValueatLevel (Variables|Default|GetStartingHeight)))))) (+ (* (Math|Vector|RotateVectorAroundAxis (Variables|Default|GetAirStrikeLineDirection) (* (Variables|Default|GetRocketsFiredNumber) (Math|Conversions|ToFloat(Integer) (/ 360 6))) "0, 0,1.000000") (* (ScalableFloat|GetValueatLevel (Variables|Default|GetDistanceBetweenRockets)) 1.0)) (Math|Vector|MakeVector (.x (Variables|Default|GetAirStrikeMidPoint)) (.y (Variables|Default|GetAirStrikeMidPoint)) (+ (.z (Variables|Default|GetAirStrikeMidPoint)) (ScalableFloat|GetValueatLevel (Variables|Default|GetStartingHeight)))))) (* (Utilities|RandomVectorinRange "-100.000000,-100.000000,0.000000" "100.000000,100.000000,0") (ScalableFloat|GetValueatLevel (Variables|Default|GetRandomLocationStrength))))))

(fn CalcSpawnRotForAllRockets
  (return (Math|Conversions|RotationFromXVector (- (Math|Vector|MakeVector (.x (Variables|Default|GetStartLocationRocket)) (.y (Variables|Default|GetStartLocationRocket)) (+ (.z (Variables|Default|GetStartLocationRocket)) -5e+03)) (Variables|Default|GetStartLocationRocket)))))

(fn FireAirStrikeRocket
  (params (StartLoc :type Vector) (StartRot :type Rotator) (ECRocket :type Fort%20Gameplay%20Effect%20Container%20Spec%20Structure) (InitialSpeed :type Float%20%28double-precision%29) (GravityScale :type Float%20%28double-precision%29))
  (bind _returnvalue (Projectile|SpawnProjectile :ProjectileClass (Variables|Default|GetPrj_AirStrikeRocket) :RequestedBy (Variables|Actor|GetInstigator) :SpawnLocation StartLoc :SpawnRotation StartRot :SpawnDirection StartRot :EffectContainerSpecToApplyOnExplode ECRocket :bUseDefaultPhysics false :InitialSpeed InitialSpeed :GravityScale GravityScale :ChargePercent 0.0))
  (Delivery|AddAdditionalExplodeEffectContainer _returnvalue (Variables|Default|GetECRocket_Additional)))

(fn OnRep_StoredHit
  (bind (_bblockinghit _binitialoverlap _time _distance _location) (Collision|BreakHitResult (Variables|Default|GetStoredHit)))
  (Variables|Default|SetHitLocation _location))

(fn ConstructionScript)
