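; Grenade-with-trajectory gameplay ability for the Air Strike (AppleSauce) consumable, extending a parent ability that spawns the thrown projectile.
; When a projectile spawns, it calls the parent, casts to the AppleSauce grenade and hands it two effect container specs.
; The rocket spec comes from the ability's player damage container, and the additional rocket spec from the generic projectile additional-explosion tag.
; The end-ability override is disabled and SetBalanceValues is empty.
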(event Ability|EventOnEndAbility
  (params (bWasCancelled :type Boolean))
  (|CallParentFunction bWasCancelled) ; This node is disabled and will not be called.\nDrag off pins to build functionality.
)

(event ProjectileLifecycle|EventOnProjectileSpawned
  (params (NewSpawnedProjectile :type Fort%20Projectile%20Base%20Object%20Reference))
  (|CallParentFunction NewSpawnedProjectile)
  (bind _returnvalue (Ability|MakeGameplayEffectContainerSpecfromAbility (Variables|Default|GetEC%20PlayerDamage) self))
  (bind _asb_prj_athena_apple_sauce_grenade (Utilities|Casting|CastToB_Prj_Athena_AppleSauce_Grenade NewSpawnedProjectile))
  (Variables|/Game/Athena/Items/Consumables/AirStrike/B_Prj_Athena_AppleSauce_Grenade.B_Prj_Athena_AppleSauce_Grenade_C|SetECRocket _returnvalue _asb_prj_athena_apple_sauce_grenade)
  (Variables|/Game/Athena/Items/Consumables/AirStrike/B_Prj_Athena_AppleSauce_Grenade.B_Prj_Athena_AppleSauce_Grenade_C|SetECRocket_Additional (Ability|MakeGameplayEffectContainerSpec "(TagName=\"Abilities.Generic.GenericProjectile.Explode.Additional\")") _asb_prj_athena_apple_sauce_grenade))

(fn SetBalanceValues)
