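# Translucent, two-sided unlit mesh-particle material for the AirStrike consumable's arc: a smoke trail with a missile sprite traveling along it.
# Two panning smoke textures build the time-of-day-tinted body, while AnimationOffset plus U drives a curve atlas row for noise-flickered glow and opacity.
# The same offset slides a windowed missile texture over the trail.
# World position offset adjusts strip width, varying it along U.
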Material /Game/Athena/Items/Consumables/AirStrike/M_AirStrikeArc
graph {
 properties {
  bUseMaterialAttributes=true
  bUsedWithMeshParticles=true
  blendMode=BLEND_Translucent
  pixelDepthOffsetMode=PDOM_Legacy
  shadingModel=MSM_Unlit
  twoSided=true
  usedShadingModels=Unlit
 }
 nodes [
  Constant()
  n1 = Desaturation(
   None=MaterialFunctionCall() {materialFunction=&Color_Time_of_Day}.Result
   Fraction=Constant() {r=0.35}
  )
  n2 = TextureCoordinate() {coordinateIndex=1}
  n3 = Add(A=TextureCoordinate(), B=AppendVector(A=ComponentMask(None=@n2) {g=true}, B=Constant()))
  n4 = Constant2Vector() {g=1, r=1}
  n5 = TextureSample(UVs=Panner(Coordinate=Multiply(A=@n3, B=@n4)) {speedX=-0.008}) {texture=/Game/Effects/Fort_Effects/Textures/Smoke/T_Puffy_Smoke_Trail_01}
  n6 = LinearInterpolate(
   A=VectorParameter() {parameterName=Color2, defaultValue=<rgba 0.1354 0.1354 0.1354 0>}.RGB
   B=VectorParameter() {parameterName=Color1, defaultValue=<rgba 0.3683 0.4904 0.5521 0>}.RGB
   Alpha=@n5.R
  )
  n7 = Multiply(A=@n3, B=Constant2Vector() {g=1, r=4})
  n8 = Multiply(
   A=LinearInterpolate(Alpha=@n5.R) {constA=0.5}
   B=TextureSample(UVs=Panner(Coordinate=@n7) {speedX=-0.08}) {samplerType=SAMPLERTYPE_Grayscale, texture=&T_Smoke_Trail_Soft}.R
  )
  n9 = Multiply(A=VertexColor().R, B=ComponentMask(None=TextureCoordinate()) {r=true})
  n10 = GetMaterialAttributes(None=SetMaterialAttributes!(MaterialAttributes, Emissive Color=Multiply(A=@n1, B=@n6), Opacity=Multiply(A=@n8, B=@n9)) {attributeSetTypes [#1, #2]}) {attributeGetTypes [#1, #2]}
  n11 = Add(
   A=ComponentMask(None=TextureCoordinate()) {r=true}
   B=DynamicParameter() {defaultValue=<rgba 0.1 1 1 1>, paramNames [AnimationOffset, Param2, Param3, Param4]}.AnimationOffset
  )
  n12 = CurveAtlasRowParameter(CurveTime=@n11) {
   atlas=&AthenaCurveAtlas01
   curve=&/Curve_Ramps
   defaultValue=108
   parameterName=UV
  }
  n13 = Time()
  n14 = Constant2Vector() {g=0.0031, r=0.2}
  n15 = ComponentMask(None=TextureCoordinate()) {r=true}
  n16 = TextureSample(UVs=Add(A=Multiply(A=@n13, B=@n14), B=Multiply(A=@n15) {constB=0.1})) {samplerType=SAMPLERTYPE_Normal, texture=/Game/Global/Textures/Default/Noise/64x64Noise}
  n17 = TextureSample() {texture=/Game/Global/Textures/Default/Masks_and_Gradients/T_Water_Top_Ramp}
  n18 = Multiply(
   A=Multiply(
    A=VectorParameter() {parameterName=Emissive, defaultValue=<rgba 12 1.704 0.12 0>}.RGB
    B=LinearInterpolate(Alpha=@n16.R) {constA=8, constB=16}
   )
   B=Multiply(A=@n17.R, B=@n17.R)
  )
  n19 = Add(A=@n10.Emissive Color, B=Multiply(A=@n10.Emissive Color, B=Multiply(A=@n12.G, B=@n18)))
  n20 = Add(A=Multiply(A=Multiply(A=@n11) {constB=24}) {constB=-1}, B=ScalarParameter() {defaultValue=24.5, parameterName=MissileOffset})
  n21 = AppendVector(A=@n20, B=ComponentMask(None=TextureCoordinate()) {g=true})
  n22 = TextureSample(UVs=@n21) {samplerSource=SSM_Clamp_WorldGroupSettings, texture=/Game/Effects/Fort_Effects/Textures/Misc/T_Missile}
  n23 = Multiply(
   A=MaterialFunctionCall() {materialFunction=&Color_Time_of_Day}.Result
   B=Multiply(A=@n22.RGB) {constB=0.5}
  )
  n24 = Abs(None=Multiply(A=Add(A=@n20) {constB=-0.5}) {constB=2})
  n25 = Multiply(A=@n22.A, B=OneMinus(None=Floor(None=@n24)))
  n26 = Constant() {r=1}
  n27 = Multiply(A=@n10.Opacity, B=Multiply(A=@n12.B, B=ParticleColor().A))
  n28 = TextureCoordinate() {uTiling=2, vTiling=2}
  n29 = Multiply(A=Transform(None=Constant3Vector() {constant=<rgba 0 1 0 0>}), B=ComponentMask(None=Subtract(A=@n28)) {g=true})
  n30 = LinearInterpolate(Alpha=ComponentMask(None=TextureCoordinate()) {r=true}) {constA=-500, constB=-200}
  SetMaterialAttributes!(
   MaterialAttributes=SetMaterialAttributes!(MaterialAttributes, Emissive Color=LinearInterpolate(A=@n19, B=@n23, Alpha=@n25), Opacity=LinearInterpolate(A=@n27, B=@n26, Alpha=@n25)) {attributeSetTypes [#1, #2]}
   World Position Offset=Divide(A=Multiply(A=@n29, B=@n30)) {constB=200}
  ) {attributeSetTypes [#3]}
  TextureSample(UVs=@n21) {texture=/Game/Effects/Fort_Effects/Textures/Ramps/T_Battery_Charge_Ramp}
 ]
}
