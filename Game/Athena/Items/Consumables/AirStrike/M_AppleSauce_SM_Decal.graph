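# Unlit translucent material for Niagara mesh particles that draws an air strike decal on scene geometry.
# World position behind translucency is projected onto the object's axes as decal coordinates, avoiding a true decal.
# These shape scrolling concentric rings, a pulsing center dot, edge falloff and a height fade into an opacity mask.
# Particle color tints it and a Custom node discards low alpha; a cloud-noise smoke branch appears left over.
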Material /Game/Athena/Items/Consumables/AirStrike/M_AppleSauce_SM_Decal
graph {
 properties {
  bCastRayTracedShadows=false
  bDisableDepthTest=true
  bUsedWithNiagaraMeshParticles=true
  blendMode=BLEND_Translucent
  pixelDepthOffsetMode=PDOM_Legacy
  shadingModel=MSM_Unlit
  usedShadingModels=Unlit
 }
 nodes [
  n1 = Length(None=NamedRerouteUsage()) {is=class MaterialExpressionLength}
  n2 = Multiply(A=@n1, B=ScalarParameter() {defaultValue=20, parameterName=Concentric Ring Tile}) {constB=2}
  n3 = Multiply(A=Time(), B=ScalarParameter() {defaultValue=2, parameterName=Time Scale})
  n4 = Sine(None=Add(A=Power(Base=@n2) {constExponent=1}, B=@n3))
  Abs(None=@n4)
  n5 = NamedRerouteUsage()
  n6 = Constant()
  n7 = SphereMask(A=@n5, B=@n6) {attenuationRadius=0.5, hardnessPercent=0}
  n8 = SphereMask(A=@n5, B=@n6, Radius=ScalarParameter() {defaultValue=0.7, parameterName=Center Dot Radius}) {attenuationRadius=0.02, hardnessPercent=0.25}
  Add(A=Multiply(A=@n7) {constB=0.75}, B=@n8)
  ComponentMask() {g=true, r=true}
  n9 = PreSkinnedLocalBounds()
  ComponentMask(None=@n9.Half Extents) {g=true, r=true}
  Constant3Vector() {constant=<rgba 1 0 0 0>}
  Constant()
  n10 = ParticleColor()
  MaterialExpressionMultiply_18 = Multiply(A=@n10.A, B=NamedRerouteUsage())
  MaterialExpressionScalarParameter_20 = ScalarParameter() {defaultValue=0.01, parameterName=Clip Value}
  Custom!(Alpha=@MaterialExpressionMultiply_18, Clip=@MaterialExpressionScalarParameter_20) {
   code="if ( Alpha < Clip ) discard;\r\nreturn Alpha;"
   description=Clip Alpha Mask
   outputType=CMOT_Float1
   inputs [
    {inputName=Alpha, input {expression=&self:MaterialExpressionMultiply_18}}
    {inputName=Clip, input {expression=&self:MaterialExpressionScalarParameter_20}}
   ]
  }
  n11 = Divide(A=ComponentMask(None=VertexNormalWS()) {b=true}) {constB=256}
  n12 = TextureObjectParameter() {parameterName=CloudTex, samplerType=SAMPLERTYPE_LinearGrayscale, texture=/Game/Packages/Fortress_Sky/SkyDome/Master/T_SkyDome_Cloud_Voroni}
  n13 = Multiply(A=TextureCoordinate(), B=DynamicParameter() {paramNames [TextureSize,Param2,Param3,Param4]}.TextureSize)
  n14 = Constant2Vector() {g=1.2, r=3}
  n15 = TextureSample(UVs=Panner(Coordinate=Multiply(A=@n13, B=@n14)) {speedX=0.01, speedY=0.1}, Tex=@n12) {samplerType=SAMPLERTYPE_LinearColor, texture=/Game/Packages/Fortress_Sky/SkyDome/Master/T_SkyDome_CloudShapes}
  n16 = SceneDepth()
  n17 = ScalarParameter() {defaultValue=750, parameterName=FakeDepthForLow}
  n18 = FeatureLevelSwitch(Default=@n16.R, ES2=@n17, ES3_1=@n17, SM4_REMOVED=@n16.R, SM5=@n16.R, SM6=@n16.R)
  n19 = Divide(A=@n18) {constB=1024}
  n20 = Add(A=Divide(A=Saturate(None=@n19)) {constB=32}, B=@n13)
  n21 = Constant2Vector() {g=0.5, r=1}
  n22 = Add(A=Multiply(A=@n15.B) {constB=0.1}, B=Panner(Coordinate=Multiply(A=@n20, B=@n21)) {speedX=-0.01, speedY=0.1})
  n23 = Subtract(A=Sine(None=TextureCoordinate() {uTiling=0.5, vTiling=0.5})) {constB=0.2}
  n24 = LinearInterpolate(
   A=Multiply(A=TextureSample(UVs=@n11, Tex=@n12) {samplerType=SAMPLERTYPE_LinearColor, texture=&T_SkyDome_CloudShapes}.B) {constB=0.5}
   B=Multiply(
    A=@n15.B
    B=TextureSample(UVs=@n22, Tex=@n12) {samplerType=SAMPLERTYPE_LinearColor, texture=&T_SkyDome_CloudShapes}.B
   )
   Alpha=Saturate(None=ComponentMask(None=@n23) {g=true})
  )
  n25 = VertexNormalWS()
  n26 = CameraVectorWS()
  n27 = SquareRoot(None=Abs(None=DotProduct(A=@n25, B=@n26)))
  n28 = OneMinus(None=SphereMask(A=WorldPosition().XYZ, B=CameraPositionWS(), Radius=Constant() {r=512}) {hardnessPercent=0})
  n29 = Subtract(A=SquareRoot(None=Add(A=@n24) {constB=0.2}), B=OneMinus(None=Multiply(A=@n27, B=@n28)))
  n30 = ScalarParameter() {defaultValue=256, parameterName=CloseFade}
  n31 = Distance(A=WorldPosition().XYZ, B=CameraPositionWS())
  n32 = Distance(A=@n9.Min, B=@n9.Min)
  n33 = Divide(
   A=Distance(A=ObjectPositionWS(), B=CameraPositionWS())
   B=Multiply(
    A=Multiply(A=@n32) {constB=0.6}
    B=MaterialFunctionCall() {materialFunction=&ObjectScale}.Scale X
   )
  ) {constB=4096}
  n34 = ScalarParameter() {parameterName=InsideShowsUpBIse}
  n35 = LinearInterpolate(
   A=Saturate(None=Divide(A=@n18, B=@n30))
   B=Divide(A=Subtract(A=@n18, B=@n31), B=ScalarParameter() {defaultValue=256, parameterName=EdgeDistance})
   Alpha=Saturate(None=Add(A=@n33, B=@n34) {constA=0.25, constB=-0.25})
  ) {constA=1}
  n36 = ScalarParameter() {parameterName=NoneEdgeEnhance}
  n37 = ScalarParameter() {defaultValue=1.5, parameterName=EdgeEnhance}
  n38 = SphereMask(A=@n31, B=@n18) {hardnessPercent=0}
  n39 = Saturate(None=Multiply(A=@n38, B=@n38))
  n40 = Clamp(None=Multiply(A=Saturate(None=@n35), B=LinearInterpolate(A=@n36, B=@n37, Alpha=@n39) {constA=1})) {clampMode=CMODE_ClampMax, maxDefault=4}
  n41 = CurveAtlasRowParameter(CurveTime=Divide(A=Multiply(A=@n29, B=@n40) {constA=1})) {atlas=&AthenaCurveAtlas01, curve=&/Curve_RedSmoke, defaultValue=110}
  Desaturation(None=@n41, Fraction=Constant() {r=0.05})
  n42 = Power(Base=NamedRerouteUsage())
  MaterialFunctionCall(Emissive=Multiply(A=@n10.RGB, B=LinearInterpolate(Alpha=@n42) {constA=0.2})) {materialFunction=&MF_FX_EmissiveAdjustments}
  Multiply(A=@n10.A, B=@n41.A)
  Multiply(A=MaterialFunctionCall() {materialFunction=&Color_Time_of_Day}.Result)
  n43 = Multiply(A=ComponentMask(None=NamedRerouteUsage()) {b=true}) {constB=2}
  n44 = Power(Base=Saturate(None=OneMinus(None=@n43))) {constExponent=4}
  n45 = Power(Base=OneMinus(None=@n7), Exp=ScalarParameter() {defaultValue=30, parameterName=Inner Falloff Power}) {constExponent=1}
  n46 = Saturate(None=@n4)
  n47 = ScalarParameter() {defaultValue=2, parameterName=Concentric Ring Power}
  n48 = ScalarParameter() {defaultValue=0.66, parameterName=Center Point}
  n49 = Subtract(A=Multiply(A=@n1) {constB=2}, B=@n48)
  n50 = OneMinus(None=@n48)
  n51 = Multiply(
   A=Multiply(A=Power(Base=@n46, Exp=@n47) {constExponent=4}, B=ScalarParameter() {defaultValue=0.33, parameterName=Concentric Ring Opacity}) {constB=0.4}
   B=Power(Base=Divide(A=@n49, B=@n50), Exp=ScalarParameter() {defaultValue=1.5, parameterName=Concentric Ring Inner Falloff Power}) {constExponent=1}
  )
  n52 = Multiply(A=ConstantBiasScale(None=Sine(None=@n3) {period=2}) {bias=1.1}, B=ScalarParameter() {defaultValue=0.5, parameterName=Center Dot Multiplier})
  n53 = Multiply(
   A=Add(A=Add(A=@n45, B=@n51), B=Multiply(A=@n8, B=@n52))
   B=SphereMask(A=@n5, B=@n6, Hardness=ScalarParameter() {defaultValue=0.99, parameterName=Outter Falloff Hardness}) {attenuationRadius=0.5, hardnessPercent=0.95}
  ) {constA=1}
  NamedRerouteDeclaration(None=Saturate(None=Multiply(A=@n44, B=@n53) {constA=1})) {name=Opacity, nodeColor=<rgba 1 0 0.0937 1>}
  n54 = MaterialFunctionCall() {materialFunction=/Engine/Functions/Engine_MaterialFunctions02/WorldPositionOffset/ComponentBasisVectors}
  n55 = Subtract(
   A=MaterialFunctionCall() {materialFunction=&WorldPositionBehindTranslucency}.Result
   B=MaterialFunctionCall() {materialFunction=&ObjectPivotPoint}.Object Pivot Location
  )
  n56 = AppendVector(A=DotProduct(A=Normalize(VectorInput=@n54.X), B=@n55), B=DotProduct(A=Normalize(VectorInput=@n54.Y), B=@n55))
  n57 = DotProduct(A=Normalize(VectorInput=@n54.Z), B=@n55)
  n58 = MaterialFunctionCall() {materialFunction=/Engine/Functions/Engine_MaterialFunctions02/WorldPositionOffset/ObjectScale}
  n59 = ScalarParameter() {defaultValue=95, parameterName=Dimensions}
  NamedRerouteDeclaration(None=Divide(A=AppendVector(A=@n56, B=@n57), B=Multiply(A=@n58.Scale XYZ, B=@n59) {constB=100})) {name=Decal Coords, nodeColor=<rgba 1 0.9141 0 1>}
  NamedRerouteDeclaration(None=@n51) {name=Concentric Rings, nodeColor=<rgba 0.0078 1 0 1>}
  ScalarParameter() {defaultValue=3000, parameterName=ShowInside}
  ScalarParameter() {defaultValue=256, parameterName=EdgeDistance_Sphere}
  ScalarParameter() {defaultValue=3000, parameterName=ShowInside}
  SceneDepth()
  SceneDepth()
  TextureSample() {samplerType=SAMPLERTYPE_LinearGrayscale, texture=&T_SkyDome_Cloud_Voroni}
 ]
}
