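# Niagara red smoke effect for the Athena Air Strike consumable.
# Mesh_Radius bursts a red mesh after a short delay, which scales up and fades, apparently as a radius indicator.
# Smoke runs two spawn-rate groups with masked lifetime, offset and lift, shaped by emitter-age curves, turning sprites from red to dark grey through two renderers and animated material parameters.
# No module shown reads the exposed User parameters.
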NiagaraSystem /Game/Athena/Items/Consumables/AirStrike/NS_Athena_AirStrike_RedSmoke
dependencies {
 usedDataInterfaces [&NiagaraDataInterfaceCurve]
 usedDynamicInputs [&Vector2DFromFloat]
 usedModules [&SystemState]
}
emitters {
 Mesh_Radius {
  emitterUpdateScript [
   &EmitterState()
   &SpawnBurst_Instantaneous(Spawn Time=0.5, Use Spawn Probability=false, UseLoopCountLimit=false)
  ]
  particleSpawnScript [
   &InitializeParticle(
    Color=<rgba 1 0 0 1>
    Lifetime=5
    Mesh Scale=<30 30 30>
    Mesh Scale Mode=::NewEnumerator4
    UsePositionOffset=false
    Write Material Random=true
   )
   &AddVelocity(Velocity=&RandomRangeVector(Minimum=<-10 -10 50>, Maximum=<10 10 100>)) {enabled=false}
   &DynamicMaterialParameters(
    Index 0 Param 1=1
    Module.Param0WriteEnabled=true
    Module.Param1WriteEnabled=false
    Module.Param2WriteEnabled=false
    Module.Param3WriteEnabled=false
   )
  ]
  particleUpdateScript [
   &ParticleState()
   &ScaleMeshSize(Scale Factor=&VectorFromFloat(Value=&FloatFromCurve(
      FloatCurve=DataInterface() {propertyValues {Curve {keys [
          {arriveTangent=1.8786, interpMode=RCIM_Cubic, time=0, value=0.66}
          {arriveTangent=1.36, interpMode=RCIM_Cubic, leaveTangent=1.36, time=0.1, value=0.925}
          {interpMode=RCIM_Cubic, tangentMode=RCTM_User, time=0.25, value=1}
          {arriveTangent=-0.0667, interpMode=RCIM_Cubic, leaveTangent=-0.0667, time=0.9, value=1}
          {arriveTangent=-0.6685, interpMode=RCIM_Cubic, leaveTangent=-0.6685, tangentMode=RCTM_User, time=1, value=0.95}
         ]}}}
      Scale Curve=0.61
     )))
   &SolveForcesAndVelocity(Clamp Velocity=false, Limit Acceleration=false) {enabled=false}
   &Color(Color=&MakeLinearColorFromVectorAndFloat(
     "Vector (RGB)"=<0.66 0 0>
     "Float (Alpha)"=&FloatFromCurve(
      FloatCurve=DataInterface() {propertyValues {Curve {keys [
          <at 0 0 RCIM_Cubic>
          {interpMode=RCIM_Cubic, tangentMode=RCTM_User, time=0.1, value=1}
          {interpMode=RCIM_Cubic, tangentMode=RCTM_User, time=0.95, value=1}
          <at 1 0 RCIM_Cubic>
         ]}}}
      Scale Curve=0.7
     )
    ))
  ]
  rendererClasses [&NiagaraMeshRendererProperties]
  renderers [{rendererClass=&NiagaraMeshRendererProperties, rendererIndex=0}]
 }
 Smoke {
  emitterUpdateScript [
   &EmitterState()
   &SpawnBurst_Instantaneous(Use Spawn Probability=false, UseLoopCountLimit=true)
   &SpawnRate(SpawnRate=10, Use Spawn Probability=false)
   &SpawnRate(SpawnGroup=1, SpawnRate=6, Use Spawn Probability=false) {module=SpawnRate001}
  ]
  particleSpawnScript [
   &InitializeParticle(
    Color=<rgba 0.5 0.5 0.5 1>
    Lifetime=&MaskFloatBySpawnGroup(
     Masked Float=&RandomRangeFloat(Minimum=2.75, Maximum=3)
     Passthrough Float=&RandomRangeFloat(Minimum=1.5, Maximum=2)
    )
    Position Offset=&MaskVectorBySpawnGroup(Masked Vector=<0 0 0>, Passthrough Vector=<0 0 50>)
    Sprite Rotation Angle=&Add_Float(A=45, B=&RandomRangeFloat(Minimum=-15, Maximum=15))
    Sprite Rotation Mode=::NewEnumerator0
    Sprite Size Mode=::NewEnumerator3
    Uniform Sprite Size=&Multiply_Float(
     A=&RandomRangeFloat(Minimum=280, Maximum=300)
     B=&FloatFromCurve(
      FloatCurve=DataInterface() {propertyValues {Curve {keys [
          {interpMode=RCIM_Cubic, tangentMode=RCTM_User, time=0.5, value=1}
          {arriveTangent=-1.4243, interpMode=RCIM_Cubic, leaveTangent=-1.4243, tangentMode=RCTM_User, time=1, value=0.4}
         ]}}}
      CurveIndex=$Emitter.NormalizedLoopAge
     )
    )
    UsePositionOffset=true
    Write Material Random=true
   ) {module=InitializeParticle001}
   &ShapeLocation(
    Hemisphere Distribution=<1 0.7>
    MaskSpawnGroup=true
    Override Seed=false
    Spawn Group Mask=1
    Sphere Radius=10
    Sphere Surface Distribution=1
    UseOwnerScale=true
   )
   &AddVelocity(
    Rotation Coordinate Space=::World
    Velocity=&MaskVectorBySpawnGroup(
     Masked Vector=&RandomRangeVector(Minimum=<-5 -5 2>, Maximum=<5 5 5>)
     Passthrough Vector=&RandomRangeVector(Minimum=<-10 -10 2>, Maximum=<10 10 4>)
    )
    Velocity Speed Scale=10
   )
  ]
  particleUpdateScript [
   &ParticleState()
   &Drag(Drag=0.75, Use Linear Drag=true, Use Rotational Drag=true)
   &AccelerationForce(Acceleration=&Multiply_VectorByFloat(
     Vector=&MaskVectorBySpawnGroup(Masked Vector=<0 0 600>, Passthrough Vector=<0 0 50>)
     Float=&FloatFromCurve(
      FloatCurve=DataInterface() {propertyValues {Curve {keys [{interpMode=RCIM_Cubic, tangentMode=RCTM_User, time=0, value=1}, {interpMode=RCIM_Cubic, tangentMode=RCTM_User, time=1, value=0.33}]}}}
      Scale Curve=&FloatFromCurve(
       FloatCurve=DataInterface() {propertyValues {Curve {keys [
           <at 0 0.15 RCIM_Cubic>
           <at 1.25 1>
           {interpMode=RCIM_Cubic, tangentMode=RCTM_User, time=4.5, value=1}
           {interpMode=RCIM_Cubic, tangentMode=RCTM_User, time=5, value=0.25}
          ]}}}
       CurveIndex=$Emitter.Age
      )
     )
    ))
   &SpriteRotationRate(Rotation Rate=&Multiply_Float(
     A=&UniformRangedFloat(Minimum=-1, Maximum=1)
     B=&FloatFromCurve(
      FloatCurve=DataInterface() {propertyValues {Curve {keys [
          {arriveTangent=-1.6805, interpMode=RCIM_Cubic, leaveTangent=-1.6805, tangentMode=RCTM_User, time=0, value=1}
          {interpMode=RCIM_Cubic, tangentMode=RCTM_User, time=0.66, value=0}
         ]}}}
      Scale Curve=350
     )
    ))
   &SolveForcesAndVelocity(Clamp Velocity=false, Limit Acceleration=false)
   &CameraOffset(Camera Offset Amount=&Multiply_Float(A=&MakeFloatFromVector2d(Vector 2D=$Particles.SpriteSize), B=0.25))
   &ScaleSpriteSize(Uniform Curve Sprite Scale=DataInterface() {propertyValues {Curve {keys [
        {interpMode=RCIM_Cubic, tangentMode=RCTM_User, time=0, value=0}
        {arriveTangent=1.5164, interpMode=RCIM_Cubic, leaveTangent=1.5164, tangentMode=RCTM_User, time=0.3, value=0.65}
        {interpMode=RCIM_Cubic, tangentMode=RCTM_User, time=1, value=1}
       ]}}})
   &ScaleColor(
    Scale Alpha=&FloatFromCurve(
     FloatCurve=DataInterface() {propertyValues {Curve {keys [
         {arriveTangent=14.4467, interpMode=RCIM_Cubic, leaveTangent=14.4467, tangentMode=RCTM_User, time=0.1, value=0}
         {interpMode=RCIM_Cubic, tangentMode=RCTM_User, time=0.2, value=1}
         {interpMode=RCIM_Cubic, tangentMode=RCTM_User, time=0.5, value=1}
         {interpMode=RCIM_Cubic, leaveTangent=-1.9727, time=1, value=0}
        ]}}}
     Scale Curve=0.7
    )
    Scale RGB=&Lerp_Vector(
     A=<2 0 0>
     B=<0.1 0.1 0.1>
     Alpha=&FloatFromCurve(FloatCurve=DataInterface() {propertyValues {Curve {keys [{arriveTangent=3.1337, interpMode=RCIM_Cubic, leaveTangent=3.1337, tangentMode=RCTM_User, time=0, value=0}, <at 1 1 RCIM_Cubic>]}}})
    )
    ScaleA=true
    ScaleRGB=true
   ) {module=ScaleColor001}
   &DynamicMaterialParameters(
    Index 0 Param 1=&Add_Float(
     A=&RandomRangeFloat(Minimum=0, Maximum=16)
     B=&FloatFromCurve(
      FloatCurve=DataInterface() {propertyValues {Curve {keys [
          {arriveTangent=-2.2561, interpMode=RCIM_Cubic, leaveTangent=-2.2561, tangentMode=RCTM_User, time=0, value=1}
          {arriveTangent=-1.0036, interpMode=RCIM_Cubic, leaveTangent=-1.0036, tangentMode=RCTM_User, time=0.2682, value=0.4215}
          <at 1 0>
         ]}}}
      Scale Curve=-0.75
     )
    )
    Index 0 Param 3=&FloatFromCurve(FloatCurve=DataInterface() {propertyValues {Curve {keys [{arriveTangent=-17.6048, interpMode=RCIM_Cubic, leaveTangent=-17.6048, tangentMode=RCTM_User, time=0, value=3}, <at 0.25 0 RCIM_Cubic>]}}})
    Index 1 Param 1=&FloatFromCurve(
     FloatCurve=DataInterface() {propertyValues {Curve {keys [{arriveTangent=-14.778, interpMode=RCIM_Cubic, time=0, value=1}, {interpMode=RCIM_Cubic, leaveTangent=-3.9602, time=0.2, value=1}]}}}
     Scale Curve=0
    )
    Index 1 Param 2=&FloatFromCurve(
     FloatCurve=DataInterface() {propertyValues {Curve {keys [{tangentMode=RCTM_User, time=0, value=0}, {tangentMode=RCTM_User, time=1, value=1}]}}}
     Scale Curve=-3
    )
    Index 1 Param 3=1
    Index 1 Param 4=&FloatFromCurve(
     FloatCurve=DataInterface() {propertyValues {Curve {keys [{arriveTangent=-5.8231, time=0, value=1.5}, {leaveTangent=-3.9602, time=0.2, value=0.5}]}}}
     Scale Curve=1.5
    )
    Index 2 Param 1=&FloatFromCurve(FloatCurve=DataInterface() {propertyValues {Curve {keys [{arriveTangent=2.5015, interpMode=RCIM_Cubic, time=0.25, value=0}, {interpMode=RCIM_Cubic, leaveTangent=1.688, time=0.75, value=1}]}}})
    Index 2 Param 2=&Add_Float(
     A=&RandomRangeFloat(Minimum=-1, Maximum=1)
     B=&FloatFromCurve(
      FloatCurve=DataInterface() {propertyValues {Curve {keys [{tangentMode=RCTM_User, time=0, value=1}, {tangentMode=RCTM_User, time=1, value=0}]}}}
      Scale Curve=&RandomRangeFloat(Minimum=0.6, Maximum=0.7)
     )
    )
    Index 2 Param 3=&FloatFromCurve(
     FloatCurve=DataInterface() {propertyValues {Curve {keys [
         {arriveTangent=0.6717, interpMode=RCIM_Cubic, time=0.6, value=0.8}
         {arriveTangent=0.7028, interpMode=RCIM_Cubic, leaveTangent=0.7028, tangentMode=RCTM_User, time=1.0011, value=1}
        ]}}}
     Scale Curve=1.5
    )
    Index 2 Param 4=&RandomRangeFloat(Minimum=0.8, Maximum=1)
    Module.Param0WriteEnabled=true
    Module.Param0WriteEnabled1=true
    Module.Param0WriteEnabled2=true
    Module.Param1WriteEnabled=true
    Module.Param1WriteEnabled1=false
    Module.Param1WriteEnabled2=true
    Module.Param2WriteEnabled=false
    Module.Param2WriteEnabled1=true
    Module.Param2WriteEnabled2=true
    Module.Param3WriteEnabled=true
    Module.Param3WriteEnabled1=false
    Module.Param3WriteEnabled2=true
    Write Parameter Index 1=true
    Write Parameter Index 2=true
   ) {module=DynamicMaterialParameters001}
  ]
  rendererClasses [&NiagaraSpriteRendererProperties]
  renderers [
   {rendererClass=&NiagaraSpriteRendererProperties}
   {rendererClass=&NiagaraSpriteRendererProperties, rendererIndex=1}
  ]
 }
}
summary {
 systemName=NS_Athena_AirStrike_RedSmoke
 emitters [
  {emitterName=Mesh_Radius, rendererClasses [&NiagaraMeshRendererProperties]}
  {emitterName=Smoke, rendererClasses [&NiagaraSpriteRendererProperties]}
 ]
 userVariables [
  User.Acceleration Velocity {type=Vector3f, description="", defaultValue {struct=Vector3f, value=<0 0 0>}}
  User.Flames {type=NiagaraBool}
  User.Inherit Position {type=NiagaraFloat, defaultValue=0, description=""}
  User.Inherit Velocity {type=NiagaraFloat, defaultValue=0, description=""}
  User.Smoke {type=NiagaraBool}
  User.Smoke Emissive Multiplier {type=NiagaraFloat, defaultValue=0, description=""}
  User.Smoke Lifetime {type=NiagaraFloat, defaultValue=0, description=""}
  User.Smoke Size {type=NiagaraFloat, defaultValue=0, description=""}
  User.Wind Coordinate Space {
   type=ENiagaraCoordinateSpace
   description=""
   defaultValue {
    struct=NiagaraExt_VariableValue_Enum
    value {displayName=Simulation, enum=/Script/Niagara.ENiagaraCoordinateSpace, enumName=ENiagaraCoordinateSpace::Simulation}
   }
  }
  User.Wind Direction {type=Vector3f, description="", defaultValue {struct=Vector3f, value=<0 0 0>}}
  User.Wind Velocity {type=NiagaraFloat, defaultValue=0, description=""}
 ]
}
